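; State Tree task for the Spire boss that destroys player-built structures in the path of the major spire's plunge.
; When enabled by a scalable flag, entering the state runs an async targeting preset on the major spire and kills each player-built building actor it returns, with the spire as damage causer.
; The task succeeds after targeting returns, or immediately when disabled.
; Exiting the state removes the async targeting request by its handle.
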(event EventEnterState
  (params (Transition :type State%20Tree%20Transition%20Result%20Structure%20%28by%20ref%29))
  ; Destroy all buildings on the way of the spire
  (if (ScalableFloat|GetValueasBool (Variables|Default|GetBuildingDestroyEnabled))
    (bind (_asynctaskproxy _targetinghandle) (Class|AsyncActionPerformTargeting|PerformTargetingAsyncAction (Variables|Context|GetMajorSpire) "/SpireBoss/Gameplay/Targeting/TP_MajorSpire_PlungeBuildingDestroy.TP_MajorSpire_PlungeBuildingDestroy" true)
      (:then
        (Variables|Default|SetTargetingAsyncAction _asynctaskproxy))
      (:Targeted
        (for _array_element (TargetingSystem|TargetingResults|GetTargetingResultsActors (GameInstanceSubsystems|GetTargetingSubsystem) _targetinghandle)
          (bind _asbuilding_actor (Utilities|Casting|CastToBuildingActor _array_element))
          (if (BuildingActor|IsPlayerBuilt _asbuilding_actor)
            (Gameplay|Die :self _asbuilding_actor :DamageCauser (Variables|Context|GetMajorSpire) :!class /Script/FortniteGame.BuildingActor)))
        (StateTree|FinishTask true)))
    (else
      (StateTree|FinishTask true))))

(event EventExitState
  (params (Transition :type State%20Tree%20Transition%20Result%20Structure%20%28by%20ref%29))
  (TargetingSystem|AsyncRequest|RemoveAsyncTargetingRequestwithHandle (GameInstanceSubsystems|GetTargetingSubsystem) (Targeting|GetTargetingHandle (Variables|Default|GetTargetingAsyncAction))))
