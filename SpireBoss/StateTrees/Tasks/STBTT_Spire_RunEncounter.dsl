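; StateTree task for the Spire boss that runs its encounter for as long as the owning state is active.
; Entering the state gets the Spire boss, reads its encounter manager and calls StartEncounter on it.
; Exiting the state calls EndEncounter on the same manager.
; Both events validity-check the boss and manager first, so a missing reference makes them do nothing.
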(event EventEnterState
  (params (Transition :type State%20Tree%20Transition%20Result%20Structure%20%28by%20ref%29))
  (bind _spireboss (Variables|Default|GetSpireBoss))
  (bind _encountermanager (Variables|/SpireBoss/GameplayActors/BGA_Large_Spire.BGA_Large_Spire_C|GetEncounterManager _spireboss))
  (Utilities|IsValid _spireboss
    (:Is%20Valid
      (Utilities|IsValid _encountermanager
        (:Is%20Valid
          (Class|BPSpireEncounterManager|StartEncounter _encountermanager))))))

(event EventExitState
  (params (Transition :type State%20Tree%20Transition%20Result%20Structure%20%28by%20ref%29))
  (bind _spireboss (Variables|Default|GetSpireBoss))
  (bind _encountermanager (Variables|/SpireBoss/GameplayActors/BGA_Large_Spire.BGA_Large_Spire_C|GetEncounterManager _spireboss))
  (Utilities|IsValid _spireboss
    (:Is%20Valid
      (Utilities|IsValid _encountermanager
        (:Is%20Valid
          (Class|BPSpireEncounterManager|EndEncounter _encountermanager))))))
