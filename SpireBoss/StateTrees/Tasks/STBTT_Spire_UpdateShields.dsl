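; StateTree task for the Spire boss that turns its shielded flag on or off when a state is entered.
; On EnterState it validates the Spire boss from context, then copies the task's bEnableShields value into bShouldbeShielded on the BGA_Large_Spire actor.
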(event EventEnterState
  (params (Transition :type State%20Tree%20Transition%20Result%20Structure%20%28by%20ref%29))
  (bind _spireboss (Variables|Context|GetSpireBoss))
  (Utilities|IsValid _spireboss
    (:Is%20Valid
      (Variables|/SpireBoss/GameplayActors/BGA_Large_Spire.BGA_Large_Spire_C|SetbShouldbeShielded (Variables|Default|GetbEnableShields) _spireboss))))
