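; StateTree task for the Spire boss that sets the current state of a major spire when the task's state is entered.
; On EnterState it finds the MajorSpireGameplayComponent on the context actor and calls SetCurrentState with the configurable StateToSet value.
; It does nothing when no context actor is available.
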(event EventEnterState
  (params (Transition :type State%20Tree%20Transition%20Result%20Structure%20%28by%20ref%29))
  (bind _contextactor (Variables|Context|GetContextActor)
    (:then
      (Class|MajorSpireGameplayComponent|SetCurrentState (Actor|GetComponentbyClass _contextactor "/SpireBoss/Gameplay/Components/BP_MajorSpireGameplayComponent.BP_MajorSpireGameplayComponent_C") (Variables|Default|GetStateToSet)))
    (:else)))
